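# Looping sound cue for a television prop ambience in the LimeRock content.
# A single looping wave player plays the TV loop wave directly as the root, with no mixing or randomization.
# It uses the shared Athena TV prop attenuation and ambient props sound class, with doubled volume and a short audible range.
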SoundCue /LimeRock/Assets/Audio/SFX/TV/IP_SC_LimeRock_TV_LP
root=WavePlayer_0
nodes [{id=WavePlayer_0, properties {bLooping=true, soundWaveAssetPtr=/LimeRock/Assets/Audio/SFX/TV/Wav/IP_LimeRock_TV_LP}}]
properties {
 attenuationSettings=/Game/Sounds/Attenuation/Fort_Prop_TV_Athena_Att
 maxDistance=500
 soundClassObject=&Fort_Ambient_Props_Athena
 volumeMultiplier=2
}
